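# Sound cue for the fourth Wasabi primary melee attack, a jump uppercut.
# A mixer layers two randomized wave sets: one of three main variations with slight pitch-down and volume variance, and a B layer of two variations pitched slightly up with volume variance.
# It plays in the melee weapon sound class at reduced volume, with Wasabi melee attenuation and its own concurrency limit.
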SoundCue /HighTower/Audio/Wasabi/Wasabi_Primary_Attack_04_JumpUppercut_Cue
root=Mixer_0
nodes [
 {id=Mixer_0, children [Modulator_1,Modulator_0], properties {inputVolume [1,1]}}
 {id=Modulator_0, children [Random_2], properties {pitchMax=1.1, pitchMin=1.1, volumeMax=1.1, volumeMin=0.9}}
 {id=Modulator_1, children [Random_1], properties {pitchMax=1, pitchMin=0.9, volumeMax=1.1, volumeMin=0.9}}
 {id=Random_1, children [WavePlayer_3,WavePlayer_4,WavePlayer_5], properties {bShouldExcludeFromBranchCulling=true, weights [1,1,1]}}
 {id=Random_2, children [WavePlayer_6,WavePlayer_7], properties {weights [1,1]}}
 {id=WavePlayer_3, properties {soundWaveAssetPtr=&/Wasabi_Primary_Attack_04_JumpUppercut_01}}
 {id=WavePlayer_4, properties {soundWaveAssetPtr=&/Wasabi_Primary_Attack_04_JumpUppercut_02}}
 {id=WavePlayer_5, properties {soundWaveAssetPtr=&/Wasabi_Primary_Attack_04_JumpUppercut_03}}
 {id=WavePlayer_6, properties {soundWaveAssetPtr=&/Wasabi_Primary_Attack_04_JumpUppercutB_01}}
 {id=WavePlayer_7, properties {soundWaveAssetPtr=&/Wasabi_Primary_Attack_04_JumpUppercutB_02}}
]
properties {
 attenuationSettings=/Game/Sounds/Attenuation/Fort_Wasabi_Melee_Att
 bOverrideConcurrency=true
 maxDistance=3800
 soundClassObject=&Fort_WeaponMelee
 volumeMultiplier=0.3
 attenuationOverrides {spatializationAlgorithm=SPATIALIZATION_HRTF}
 concurrencyOverrides {maxCount=6}
 concurrencySet [null]
}
